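# Editor utility Blueprint that copies the package names of the selected assets to the system clipboard.
# Its function loops over the selection, converts each object to asset data and appends the extracted name to OutputString, with CRLF line breaks between entries.
# When the loop completes, it copies the accumulated string to the clipboard.
# OutputString is never cleared within the graphs shown.
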EditorUtilityBlueprint /Game/Blueprints/Blutilities/Developers/janiececampbell/CopyPathNamesToClipboard
graphs [
 Copy Package Names To Clipboard {nodes [
   Copy Package Names To Clipboard() {is=title K2Node_FunctionEntry}
   > n1 = GetSelectedAssets()
   > n2 = For Each Loop(Array:Array of Object References=@n1) -> (Array Element:Object Reference)
   CopyStringToClipboard!(execute=@n2.Completed, InString:String=Get OutputString() -> (OutputString:String)) -> (then)
   n3 = Break Asset Data(AssetData=CreateAssetData(InAsset=@n2, bAllowBlueprintClass=false))
   n4 = "Equal (Integer)"(A=@n2.Array Index)
   Set OutputString(
    execute=@n2.LoopBody
    OutputString:String=Append!(A=Get OutputString(), B=SelectString(B="\r\n", bPickA=@n4), C:String="To String (Name)"(InName=@n3)) -> (ReturnValue)
   ) -> (Output_Get:String)
  ]}
 EventGraph {nodes [Copy Package Names To Clipboard() {is=title K2Node_CallFunction}]}
]
variables [OutputString {type=string, default=""}]
